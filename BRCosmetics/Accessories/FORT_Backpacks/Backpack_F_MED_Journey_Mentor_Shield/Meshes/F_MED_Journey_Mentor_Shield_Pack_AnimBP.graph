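# Anim Blueprint for the Journey Mentor Shield back bling that repositions its pack bone to follow the wearer's pose and movement.
# Chained additive Modify Bone nodes apply fixed offsets and scale, plus adjustments for skydiving speed, spine bend, sprinting and emotes.
# The event graph derives these from pawn state and the owner mesh's spine_05-to-spine_01 delta.
# Despite its name, `is ChoiceCrave Idle Playing` is inverted, disabling spine correction during that montage; `isCarrotCakeAlpha` is set but unused.
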AnimBlueprint /BRCosmetics/Accessories/FORT_Backpacks/Backpack_F_MED_Journey_Mentor_Shield/Meshes/F_MED_Journey_Mentor_Shield_Pack_AnimBP
target_skeleton=&/F_MED_Journey_Mentor_Shield_Pack_Skeleton
graphs [
 AnimGraph {nodes [
   n1 = "Transform (Modify) Bone - Bone: pack"(ComponentPose=Local To Component(LocalPose=Local Space Ref Pose()), Translation="0,-1,0", Rotation="0,0,0", Scale="0.9,0.9,0.9") {
    rotationMode=BMM_Additive
    scaleMode=BMM_Additive
    translationMode=BMM_Additive
    alphaBoolBlend {blendOption=Linear}
    boneToModify {boneName=pack}
    scale=<0.9 0.9 0.9>
    translation {y=-1}
   }
   n2 = "Transform (Modify) Bone - Bone: pack"!(ComponentPose=@n1, bAlphaBoolEnabled:bool, Translation="0,0,1", Rotation="0,0,0", Scale) -> (Pose) {
    alphaInputType=Bool
    rotationMode=BMM_Additive
    scaleMode=BMM_Additive
    translationMode=BMM_Additive
    alphaBoolBlend {blendInTime=0.3, blendOutTime=0.3}
    boneToModify {boneName=pack}
    translation {z=1}
   }
   n3 = "Transform (Modify) Bone - Bone: pack"!(
    ComponentPose=@n2
    Alpha:float=Get SkydivingForwardBackwardMovementAlpha() -> (SkydivingForwardBackwardMovementAlpha:double)
    Translation="0,0,0"
    Rotation="0,0,12"
    Scale
   ) -> (Pose) {
    rotationMode=BMM_Additive
    scaleMode=BMM_Additive
    translationMode=BMM_Additive
    alphaBoolBlend {blendOption=Linear}
    alphaScaleBiasClamp {bInterpResult=true}
    boneToModify {boneName=pack}
    rotation {roll=12}
   }
   n4 = "Transform (Modify) Bone - Bone: pack"(
    ComponentPose=@n3
    Alpha=Get is ChoiceCrave Idle Playing() -> (is ChoiceCrave Idle Playing:double)
    Translation=Get backpack_translation()
    Rotation=Get backpack_rotate()
   ) {rotationMode=BMM_Additive, translationMode=BMM_Additive, alphaBoolBlend {blendOption=Linear}, boneToModify {boneName=pack}}
   n5 = "Transform (Modify) Bone - Bone: pack"(ComponentPose=@n4, Alpha=Get Is Sprinting() -> (Is Sprinting:double), Translation="0.000,-1,-0.000", Rotation="0,0,0") {
    rotationMode=BMM_Additive
    translationMode=BMM_Additive
    alphaBoolBlend {blendOption=Linear}
    alphaScaleBiasClamp {bInterpResult=true, interpSpeedDecreasing=20, interpSpeedIncreasing=20}
    boneToModify {boneName=pack}
    translation {y=-1}
   }
   n6 = "Transform (Modify) Bone - Bone: pack"(ComponentPose=@n5, Translation="0,-2,0", Rotation="0,0,16") {
    rotationMode=BMM_Additive
    translationMode=BMM_Additive
    alphaBoolBlend {blendOption=Linear}
    alphaScaleBiasClamp {interpSpeedDecreasing=20, interpSpeedIncreasing=20}
    boneToModify {boneName=pack}
    rotation {roll=16}
    translation {y=-2}
   }
   n7 = "Transform (Modify) Bone - Bone: pack"(ComponentPose=@n6, Alpha=Get Front Rot Emote() -> (Front Rot Emote:double), Translation="0,0,0", Rotation="0,0,10") {
    rotationMode=BMM_Additive
    translationMode=BMM_Additive
    alphaBoolBlend {blendOption=Linear}
    alphaScaleBiasClamp {bInterpResult=true, interpSpeedDecreasing=5, interpSpeedIncreasing=7}
    boneToModify {boneName=pack}
    rotation {roll=10}
   }
   Output Pose(Result=Component To Local(ComponentPose=@n7)) {layerGroup=None}
  ]}
 EventGraph {nodes [
   Event BlueprintUpdateAnimation()
   > n1 = Sequence!(execute) -> (then_0, then_1, then_2:Exec, then_3:Exec)
   > Set Is Sprinting(Is Sprinting:double="To Float (Boolean)"(InBool=Get bIsSprinting())) -> (Output_Get:double)
   n2 = BreakTransform(InTransform=GetDeltaTransformFromRefPose(self=GetOwnerSkeletalMeshComponent(), BoneName=spine_05, BaseName=spine_01))
   n3 = BreakRotator(InRot=@n2.Rotation)
   > Set backpack_rotate(backpack_rotate=MakeRotator(Roll=MapRangeClamped(Value=@n3.Yaw, InRangeA=0, InRangeB=-30, OutRangeB=35)))
   > Set backpack_translation(backpack_translation=MakeVector(Y=MapRangeClamped(Value=@n3.Yaw, InRangeA=0, InRangeB=-30, OutRangeB=-5)))
   n4 = GetAnimInstance(self=GetOwnerSkeletalMeshComponent())
   > Is Valid(InputObject=@n4)
   n5 = Montage_IsPlaying(self=@n4, Montage=/BRCosmetics/Animation/Game/MainPlayer/Menu/BR/CustomSkins/_s39/ChoiceCrave_M/ChoiceCrave_Idle_Montage)
   > Set is ChoiceCrave Idle Playing(is ChoiceCrave Idle Playing:double="To Float (Boolean)"(InBool=NOT Boolean(A=@n5))) -> (Output_Get:double)
   > n6 = Cast To FortPlayerPawn(Object=TryGetPawnOwner())
   n7 = GetDisplayName(Object=GetSkeletalMeshForPartType(self=@n6))
   > Set isCarrotCakeAlpha(isCarrotCakeAlpha:double="To Float (Boolean)"(InBool=MatchesWildcard(SourceString=@n7, Wildcard=*M_MED_CarrotCake*, SearchCase=CaseSensitive))) -> (Output_Get:double)
   n8 = BreakRotator(InRot=@n2.Rotation)
   n9 = Branch(execute=@n1.then_1, Condition=AND Boolean(A="float >= float"(A=@n8.Yaw, B=6), B=Get bIsPlayingEmote()))
   > Set Front Rot Emote(Front Rot Emote:double=1) -> (Output_Get:double)
   Set Front Rot Emote(execute=@n9.else, Front Rot Emote=0)
   n10 = TryGetPawnOwner()
   n11 = "Absolute (Float)"(A=Dot Product!(
     A:Vector=GetVelocity!(self:Actor Object Reference=@n10) -> (ReturnValue:Vector)
     B:Vector=GetActorForwardVector!(self:Actor Object Reference=@n10) -> (ReturnValue:Vector)
    ) -> (ReturnValue:double))
   Set SkydivingForwardBackwardMovementAlpha(
    execute=@n1.then_2
    SkydivingForwardBackwardMovementAlpha:double=float * float(A:bool=Get bIsSkyDiving(), B=MapRangeClamped(Value=@n11, InRangeA=0, InRangeB=300, OutRangeA=0.5)) {is=title K2Node_PromotableOperator}
   ) -> (Output_Get:double)
  ]}
]
variables [
 backpack_rotate {type=Rotator}
 backpack_translation {type=Vector}
 Front Rot Emote {type=number}
 is ChoiceCrave Idle Playing {type=number}
 Is Sprinting {type=number}
 isCarrotCakeAlpha {type=number}
 isFastSprinting {type=number}
 SkydivingForwardBackwardMovementAlpha {type=number}
]
